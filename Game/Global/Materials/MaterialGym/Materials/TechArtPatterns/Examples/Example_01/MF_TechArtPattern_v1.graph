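# Tech-art pattern material function that layers an animated Voronoi pattern with a thin-film, metallic look over incoming material attributes.
# MF_TextureUVs turns a Voronoi UV texture into tiled, animated base and texture-based UVs for sampling.
# Time plus blurred noise drives a sine that oscillates the smoothstep threshold on a Voronoi distance-field texture, forming a changing mask.
# The layer overrides color, metallic, roughness and flattened normal before Fresnel-driven MF_ThinFilm_TechArt and blending.
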MaterialFunction /Game/Global/Materials/MaterialGym/Materials/TechArtPatterns/Examples/Example_01/MF_TechArtPattern_v1
graph {nodes [
  n1 = NamedRerouteUsage()
  n2 = SetMaterialAttributes!(
   MaterialAttributes=@n1
   BaseColor=VectorParameter() {parameterName=BaseColor_01, defaultValue=<rgba 0.1003 0.1162 0.14 1>}.RGB
   Metallic=ScalarParameter() {defaultValue=0.5, parameterName=Metallic}
   Roughness=ScalarParameter() {defaultValue=0.25, parameterName=Roughness}
   Normal=MaterialFunctionCall(Normal=NamedRerouteUsage(), Flatness=ScalarParameter() {defaultValue=0.25, parameterName=NormalFlatness}) {materialFunction=&FlattenNormal}.Result
  ) {attributeSetTypes [
    #1
    #2
    #3
    #4
   ]}
  n3 = Fresnel() {baseReflectFraction=0, exponent=1}
  n4 = Constant() {r=1}
  FunctionOutput(None=BlendMaterialAttributes(
    A=@n1
    B=MaterialFunctionCall!(Material Attributes=@n2, Fresnel=@n3, Alpha=@n4) {materialFunction=/Game/Global/MaterialFunctions/TechArt/MF_ThinFilm_TechArt}.Result
    Alpha=NamedRerouteUsage()
   ) {vertexAttributeBlendType=UseA})
  n5 = MaterialFunctionCall!(
   Texture=TextureObject() {samplerType=SAMPLERTYPE_LinearColor, texture=/Game/Global/Materials/MaterialGym/Materials/TechArtPatterns/Textures/T_Voronoi_UV}
   TilingXY=ScalarParameter() {defaultValue=5, parameterName=PatternTiling}
   OffsetXY=ComponentMask(None=VectorParameter() {parameterName=UVOffset, defaultValue=<rgba -0.001 0.0075 0 1>}.RGB) {g=true, r=true}
   AnimateOffset=StaticBool() {value=true}
   UseParallax
   Height
  ) {materialFunction=/Game/Global/Materials/MaterialGym/Materials/TechArtPatterns/MaterialFunctions/MF_TextureUVs}
  NamedRerouteDeclaration(None=@n5.BaseUVs) {name=BaseUVs, nodeColor=<rgba 0.9922 1 0 1>}
  n6 = NamedRerouteUsage()
  n7 = Add(
   A=Multiply(A=Time(), B=ScalarParameter() {defaultValue=0.2, parameterName=NoiseSpeed}) {constB=0.1}
   B=Multiply(
    A=TextureSample(UVs=@n6) {samplerType=SAMPLERTYPE_Grayscale, texture=&LowResBlurredNoise}.R
    B=ScalarParameter() {defaultValue=3, parameterName=NoiseMultiply}
   )
  )
  n8 = Saturate(None=ConstantBiasScale(None=Sine(None=@n7)))
  n9 = ScalarParameter() {defaultValue=2, parameterName=NoiseExponent}
  NamedRerouteDeclaration(None=LinearInterpolate(
    A=ScalarParameter() {defaultValue=0.05, parameterName=DF_Min}
    B=ScalarParameter() {defaultValue=1, parameterName=DF_Max}
    Alpha=Power(Base=@n8, Exp=@n9) {constExponent=0.5}
   )) {name=Noise, nodeColor=<rgba 0.8516 1 0 1>}
  NamedRerouteDeclaration(None=TextureSample(UVs=NamedRerouteUsage()) {samplerType=SAMPLERTYPE_Normal, texture=/Game/Global/Materials/MaterialGym/Materials/TechArtPatterns/Textures/T_Voronoi_N}.RGB) {name=Normal, nodeColor=<rgba 1 0 0.1641 1>}
  NamedRerouteDeclaration(None=FunctionInput() {inputType=FunctionInput_MaterialAttributes}) {name=IN_Material, nodeColor=<rgba 0 0.6719 1 1>}
  n10 = NamedRerouteUsage()
  n11 = Add(A=@n10, B=ScalarParameter() {defaultValue=0.025, parameterName=DF_Bias}) {constB=0.025}
  n12 = TextureSample(UVs=NamedRerouteUsage()) {samplerType=SAMPLERTYPE_LinearColor, texture=/Game/Global/Materials/MaterialGym/Materials/TechArtPatterns/Textures/T_Voronoi_DF}
  NamedRerouteDeclaration(None=OneMinus(None=SmoothStep(Min=@n10, Max=@n11, Value=@n12.R) {constMax=0.1})) {name=DF, nodeColor=<rgba 0.0312 0 1 1>}
  NamedRerouteDeclaration(None=@n5.TextureUVs) {name=TextureBasedUV, desc=Texture based UV, nodeColor=<rgba 0 0.4844 1 1>}
 ]}
