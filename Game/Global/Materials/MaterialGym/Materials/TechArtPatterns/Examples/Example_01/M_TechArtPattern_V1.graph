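# Default Lit material in the MaterialGym examples that demonstrates the MF_TechArtPattern_v1 function, flagged for skeletal meshes and mesh deformers.
# It uses material attributes, setting parameterized base color, metallic and roughness plus a constant flat normal, then passing them into the pattern function.
# A separate pair of character PreFX and PostFX functions is linked to each other, with no link shown to the pattern chain.
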Material /Game/Global/Materials/MaterialGym/Materials/TechArtPatterns/Examples/Example_01/M_TechArtPattern_V1
graph {
 properties {
  bUseMaterialAttributes=true
  bUsedWithMeshDeformer=true
  bUsedWithSkeletalMesh=true
  pixelDepthOffsetMode=PDOM_Legacy
  usedShadingModels=Default Lit
 }
 nodes [
  n1 = MaterialFunctionCall() {materialFunction=&BaseFN_CharacterFunction_PreFX}
  MaterialFunctionCall(AO=@n1.AO, Distance Blend=@n1.AO, Fresnel=@n1.AO, Gmap=@n1.AO) {materialFunction=&BaseFN_CharacterFunction_PostFX}
  n2 = SetMaterialAttributes!(
   MaterialAttributes
   Base Color=VectorParameter() {parameterName=BaseColor_02, defaultValue=<rgba 0.1094 0.1094 0.1094 1>}.RGB
   Metallic=ScalarParameter() {defaultValue=0.5, parameterName=Metallic_02}
   Roughness=ScalarParameter() {defaultValue=0.4, parameterName=Roughness_02}
   Normal=Constant3Vector() {constant=<rgba 0 0 1 0>}
  ) {attributeSetTypes [
    #1
    #2
    #3
    #4
   ]}
  MaterialFunctionCall!(In=@n2) {materialFunction=&/MF_TechArtPattern_v1}
 ]
}
